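# Ambient looping SoundCue for a Figment skybox, routed through the in-game SFX sound class.
# A Mixer layers two looping waves, with the second boosted in volume.
# Overridden attenuation disables distance falloff but uses a very narrow listener-focus cone, fading the sound to silence when the listener faces away from it.
# Occlusion traces on a custom channel also mute it when blocked, with interpolation smoothing both focus and occlusion changes.
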SoundCue /Figment_S03/Athena/Sounds/Events/Ambient_SkyboxLoop_Cue
root=Mixer_0
nodes [
 {id=Mixer_0, children [WavePlayer_1,WavePlayer_2], properties {inputVolume [1,1.5]}}
 {id=WavePlayer_1, properties {bLooping=true, soundWaveAssetPtr=&/ambient_skyboxloop_figment}}
 {id=WavePlayer_2, properties {bLooping=true, soundWaveAssetPtr=&/ambient_skyboxloop_w_figment}}
]
properties {
 bOverrideAttenuation=true
 maxDistance=2097152
 soundClassObject=&Fort_SFX_InGame
 volumeMultiplier=1
 attenuationOverrides {
  bAttenuate=false
  bEnableFocusInterpolation=true
  bEnableListenerFocus=true
  bEnableOcclusion=true
  focusAttackInterpSpeed=0.1
  focusAzimuth=3
  focusReleaseInterpSpeed=0.5
  nonFocusAzimuth=7
  nonFocusVolumeAttenuation=0
  occlusionInterpolationTime=2
  occlusionTraceChannel=ECC_GameTraceChannel11
  occlusionVolumeAttenuation=0
 }
}
